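# Animation Blueprint for the M_MED_Tapas cape accessory, targeting the cape's own skeleton.
# Its AnimGraph feeds a Copy Pose From Mesh node straight into the output pose, so the cape follows the pose of another mesh.
# The EventGraph is empty and there are no variables, so it adds no logic or simulation of its own in these graphs.
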AnimBlueprint /BRCosmetics/Accessories/FORT_Capes/Cape_M_MED_Tapas/Meshes/M_MED_Tapas_Cape_AnimBP
target_skeleton=&/M_MED_Tapas_Cape_Skeleton
graphs [
 AnimGraph {nodes [Output Pose(Result=Copy Pose From Mesh()) {layerGroup=None}]}
 EventGraph {links [], nodes []}
]
variables []
